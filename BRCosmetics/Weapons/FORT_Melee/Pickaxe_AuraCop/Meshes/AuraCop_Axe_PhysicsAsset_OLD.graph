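# Physics asset for the AuraCop pickaxe melee weapon cosmetic, named as an old version.
# It defines a single capsule body on the Melee_TwoHanded_Handle bone, rotated and offset along the weapon.
# There are no constraints, so the axe collides as one rigid shape with no articulated parts.
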PhysicsAsset /BRCosmetics/Weapons/FORT_Melee/Pickaxe_AuraCop/Meshes/AuraCop_Axe_PhysicsAsset_OLD
bodies {Melee_TwoHanded_Handle [{shapeType=Capsule, length=68.1684, radius=5.4392, center=<-0.2211 -7.8185 61.6984>, rotation=<rot -0.495 -90 0.339>}]}
constraints []
